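# Sound cue for mud being applied to a player, playing one of eight random waves with slight upward pitch variation through the player foley sound class.
# An IsLocal branch sends local playback through a footstep attenuation asset, while other cases use the mud third-person attenuation.
# On that non-local path, an IsEnemy branch boosts enemy volume and strongly reduces it otherwise, making enemies far more audible.
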SoundCue /MudGameplay/Audio/Mud_Applied_Player_Cue
root=Branch_4
nodes [
 {
  id=Attenuation_4
  children [SoundClass_4]
  properties {attenuationSettings=/Game/Sounds/Fort_Player/Footsteps/Settings/Fort_Player_FS_3P_Att}
 }
 {id=Attenuation_5, children [SoundClass_5], properties {attenuationSettings=/MudGameplay/Audio/Settings/Mud_Applied_3P}}
 {id=Branch_4, children [Attenuation_4, Branch_5, Branch_5], properties {boolParameterName=IsLocal}}
 {id=Branch_5, children [Modulator_3,Modulator_1,Modulator_1], properties {boolParameterName=IsEnemy}}
 {id=Modulator_0, children [Random_0], properties {pitchMax=1.15, pitchMin=1.05}}
 {id=Modulator_1, children [Attenuation_5], properties {pitchMax=1, pitchMin=1, volumeMax=0.3, volumeMin=0.3}}
 {id=Modulator_3, children [Attenuation_5], properties {pitchMax=1, pitchMin=1, volumeMax=3, volumeMin=3}}
 {
  id=Random_0
  children [WavePlayer_0,WavePlayer_1,WavePlayer_2,WavePlayer_3,WavePlayer_4,WavePlayer_5,WavePlayer_6,WavePlayer_7]
  properties {bShouldExcludeFromBranchCulling=true, weights [1,1,1,1,1,1,1,1]}
 }
 {id=SoundClass_4, children [Modulator_0], properties {soundClassOverride=&Fort_Player_Foley}}
 {id=SoundClass_5, children [Modulator_0], properties {soundClassOverride=&Fort_Player_Foley}}
 {id=WavePlayer_0, properties {soundWaveAssetPtr=&/Mud_Applied_Player_01}}
 {id=WavePlayer_1, properties {soundWaveAssetPtr=&/Mud_Applied_Player_02}}
 {id=WavePlayer_2, properties {soundWaveAssetPtr=&/Mud_Applied_Player_03}}
 {id=WavePlayer_3, properties {soundWaveAssetPtr=&/Mud_Applied_Player_04}}
 {id=WavePlayer_4, properties {soundWaveAssetPtr=&/Mud_Applied_Player_05}}
 {id=WavePlayer_5, properties {soundWaveAssetPtr=&/Mud_Applied_Player_06}}
 {id=WavePlayer_6, properties {soundWaveAssetPtr=&/Mud_Applied_Player_07}}
 {id=WavePlayer_7, properties {soundWaveAssetPtr=&/Mud_Applied_Player_08}}
]
properties {
 maxDistance=3584
 volumeMultiplier=1.8
 concurrencyOverrides {bLimitToOwner=true, maxCount=1, resolutionRule=StopLowestPriority}
 concurrencySet [/MudGameplay/Audio/Settings/Mud_Applied_Conc]
}
