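# Sound cue for mud being applied to a vehicle, playing one of eight random wave variations with pitch and volume modulation.
# An IsLocal branch sends local playback through third-person footstep attenuation, and other playback through mud-specific attenuation.
# For non-local playback, an IsEnemy branch boosts enemy volume and strongly lowers it otherwise.
# Both paths use the player foley sound class, and a mud concurrency asset governs simultaneous plays.
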SoundCue /MudGameplay/Audio/Mud_Applied_Vehicle_Cue
root=Branch_4
nodes [
 {
  id=Attenuation_4
  children [SoundClass_4]
  properties {attenuationSettings=/Game/Sounds/Fort_Player/Footsteps/Settings/Fort_Player_FS_3P_Att}
 }
 {id=Attenuation_5, children [SoundClass_5], properties {attenuationSettings=/MudGameplay/Audio/Settings/Mud_Applied_3P}}
 {id=Branch_4, children [Attenuation_4, Branch_5, Branch_5], properties {boolParameterName=IsLocal}}
 {id=Branch_5, children [Modulator_3,Modulator_1,Modulator_1], properties {boolParameterName=IsEnemy}}
 {id=Modulator_0, children [Random_1], properties {pitchMax=1.2}}
 {id=Modulator_1, children [Attenuation_5], properties {pitchMax=1, pitchMin=1, volumeMax=0.3, volumeMin=0.3}}
 {id=Modulator_3, children [Attenuation_5], properties {pitchMax=1, pitchMin=1, volumeMax=2.5, volumeMin=2.5}}
 {
  id=Random_1
  children [WavePlayer_8, WavePlayer_9, WavePlayer_10, WavePlayer_11, WavePlayer_12, WavePlayer_13, WavePlayer_14, WavePlayer_15]
  properties {weights [1,1,1,1,1,1,1,1]}
 }
 {id=SoundClass_4, children [Modulator_0], properties {soundClassOverride=&Fort_Player_Foley}}
 {id=SoundClass_5, children [Modulator_0], properties {soundClassOverride=&Fort_Player_Foley}}
 {id=WavePlayer_10, properties {soundWaveAssetPtr=&/Mud_Applied_Vehicle_03}}
 {id=WavePlayer_11, properties {soundWaveAssetPtr=&/Mud_Applied_Vehicle_04}}
 {id=WavePlayer_12, properties {soundWaveAssetPtr=&/Mud_Applied_Vehicle_05}}
 {id=WavePlayer_13, properties {soundWaveAssetPtr=&/Mud_Applied_Vehicle_06}}
 {id=WavePlayer_14, properties {soundWaveAssetPtr=&/Mud_Applied_Vehicle_07}}
 {id=WavePlayer_15, properties {soundWaveAssetPtr=&/Mud_Applied_Vehicle_08}}
 {id=WavePlayer_8, properties {soundWaveAssetPtr=&/Mud_Applied_Vehicle_01}}
 {id=WavePlayer_9, properties {soundWaveAssetPtr=&/Mud_Applied_Vehicle_02}}
]
properties {maxDistance=3584, priority=3, volumeMultiplier=1.5, concurrencySet [/MudGameplay/Audio/Settings/Mud_Applied_Conc]}
