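# Sound cue that plays the mud removal countdown sound, likely when a player's mud coating is being removed.
# A delay node holds the wave player for a fixed 0.4 seconds before the single countdown wave plays.
# It routes through the player foley sound class at reduced volume with a very large max distance.
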SoundCue /MudGameplay/Audio/Mud_Removal_Countdown_Cue
root=Delay_0
nodes [
 {id=Delay_0, children [WavePlayer_0], properties {delayMax=0.4, delayMin=0.4}}
 {id=WavePlayer_0, properties {soundWaveAssetPtr=&/Mud_Removal_Countdown}}
]
properties {maxDistance=2097152, soundClassObject=&Fort_Player_Foley, volumeMultiplier=0.4}
