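# Sound cue for player mud removal, playing one of four random waves at lowered pitch under the player foley sound class.
# An IsLocal branch sends local playback through the footstep third-person attenuation.
# Non-local or unset playback uses an IsEnemy branch, boosting enemies and quieting others, then the mud-applied third-person attenuation.
# Both paths share the random wave chain, and concurrency allows one instance per owner.
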SoundCue /MudGameplay/Audio/Mud_Removed_Player_Cue
root=Branch_4
nodes [
 {
  id=Attenuation_4
  children [SoundClass_4]
  properties {attenuationSettings=/Game/Sounds/Fort_Player/Footsteps/Settings/Fort_Player_FS_3P_Att}
 }
 {id=Attenuation_5, children [SoundClass_5], properties {attenuationSettings=/MudGameplay/Audio/Settings/Mud_Applied_3P}}
 {id=Branch_4, children [Attenuation_4, Branch_5, Branch_5], properties {boolParameterName=IsLocal}}
 {id=Branch_5, children [Modulator_3,Modulator_1,Modulator_1], properties {boolParameterName=IsEnemy}}
 {id=Modulator_0, children [Random_1], properties {pitchMax=0.9, pitchMin=0.8}}
 {id=Modulator_1, children [Attenuation_5], properties {pitchMax=1, pitchMin=1, volumeMax=0.3, volumeMin=0.3}}
 {id=Modulator_3, children [Attenuation_5], properties {pitchMax=1, pitchMin=1, volumeMax=2, volumeMin=2}}
 {
  id=Random_1
  children [WavePlayer_8, WavePlayer_9, WavePlayer_10, WavePlayer_11]
  properties {bShouldExcludeFromBranchCulling=true, weights [1,1,1,1]}
 }
 {id=SoundClass_4, children [Modulator_0], properties {soundClassOverride=&Fort_Player_Foley}}
 {id=SoundClass_5, children [Modulator_0], properties {soundClassOverride=&Fort_Player_Foley}}
 {id=WavePlayer_10, properties {soundWaveAssetPtr=&/Mud_Player_Removed_02}}
 {id=WavePlayer_11, properties {soundWaveAssetPtr=&/Mud_Player_Removed_03}}
 {id=WavePlayer_8, properties {soundWaveAssetPtr=&/Mud_Player_Removed_04}}
 {id=WavePlayer_9, properties {soundWaveAssetPtr=&/Mud_Player_Removed_01}}
]
properties {
 attenuationSettings=&Fort_Player_FS_3P_Att
 maxDistance=3584
 soundClassObject=&Fort_Player_Foley
 volumeMultiplier=2
 concurrencyOverrides {bLimitToOwner=true, maxCount=1, resolutionRule=StopLowestPriority}
}
